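; Loot-drop-on-damage building actor component for Chrome props that applies the chrome effect and swaps in chrome destruction FX.
; Once playlist data is ready, it hooks owner death and starts a chrome check for physics-fellable props.
; As a temporary fix for pre-chromed props, a timer reapplies the GameplayEffect until the Chrome tag appears, and game phase changes can restart it.
; On death, it clears default particles and fires a GameplayCue scaled by bounds.
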(event EventBeginPlay
  ;[1 1 1] Bind to death and play the proper FX when destroyed
  (|CallParentFunction)
  (Class|FortAsyncActionCurrentPlaylistReady|PlaylistDataReadyAsync
    (:OnReady
      (bind _buildingactorowner (Variables|Default|GetBuildingActorOwner)
        (:then
          (Utilities|Casting|CastToProp_PhysicsFellable_Parent _buildingactorowner
            (:then
              (CallFunction|CheckChrome) ; Ensure physics objects get the proper GEs
            )
            (:CastFailed
              _))
          (EventDispatchers|BindEventtoOnDied _buildingactorowner (EventDispatchers|CreateEvent :!event Custom|OnDied)))
        (:else)))))

(event Custom|ApplyChrome
  ;[1 0 0] Temp solution for ensuring props get their proper GEs when pre chromed
  (bind _buildingactorowner (Variables|Default|GetBuildingActorOwner)
    (:then
      (if (GameplayTags|HasMatchingGameplayTag _buildingactorowner (GameplayTags|MakeLiteralGameplayTag "(TagName=\"Gameplay.Status.Chrome\")"))
        (Variables|Default|SetbEffectApplied true)
        (else
          (GameplayEffects|ApplyGameplayEffectToSelf :self (Ability|GetAbilitySystemComponent _buildingactorowner) :GameplayEffectClass "/ChromeShared/Blueprints/GE_Athena_Lackluster_ApplyChrome_GrantComponent.GE_Athena_Lackluster_ApplyChrome_GrantComponent_C" :Level 0.0)
          (CallFunction|CheckChrome)
          (Development|PrintString (Utilities|String|Append "applied to: " (Utilities|GetDisplayName _buildingactorowner))))))
    (:else)))

(event Custom|CheckChrome
  ;[1 0 0] Temp solution for ensuring props get their proper GEs when pre chromed
  (Utilities|Time|SetTimerbyEvent (delegate Custom|ApplyChrome) 1.0))

(event Custom|GamePhaseChanged
  (params (GamePhase :type EAthenaGamePhase%20Enum))
  ;[1 0 0] Temp check for GE on phase change - issue with pre chromed assets not having the GEs
  (if (not (Variables|Default|GetbEffectApplied))
    (CallFunction|CheckChrome)))

(event Custom|RemoveDeathParticles
  ;[1 1 1] Clear out previous death FX
  (bind _asbuilding_prop (Utilities|Casting|CastToBuildingProp (Variables|Default|GetBuildingActorOwner))
    (:then
      (Class|BuildingStaticMesh|SetDeathParticles _asbuilding_prop 0))
    (:CastFailed
      (bind _asbuilding_container (Utilities|Casting|CastToBuildingContainer (Variables|Default|GetBuildingActorOwner)))
      (Class|BuildingStaticMesh|SetDeathParticles _asbuilding_container 0))))

(event Custom|OnDied
  (params (DamagedActor :type Actor%20Object%20Reference) (Damage :type Float%20%28single-precision%29) (InstigatedBy :type Controller%20Object%20Reference) (DamageCauser :type Actor%20Object%20Reference) (HitLocation :type Vector) (FHitComponent :type Primitive%20Component%20Object%20Reference) (BoneName :type Name) (Momentum :type Vector))
  ;[1 0.7758 0.4615] On BuildingActorOwner Died
  (CallFunction|RemoveDeathParticles)
  (bind _buildingactorowner (Variables|Default|GetBuildingActorOwner)
    (:then
      (Utilities|IsValid (Ability|GetAbilitySystemComponent _buildingactorowner)
        (:Is%20Valid
          (GameplayCue|ExecuteGameplayCueOnActor(Burst) _buildingactorowner (Variables|Default|GetGCN_Destroyed) (Ability|GameplayCue|MakeGameplayCueParameters :NormalizedMagnitude 0.0 :RawMagnitude (Math|Float|Clamp(Float) (/ (Math|Vector|VectorLength (.BoxExtent (Collision|GetActorBounds _buildingactorowner false false))) 2e+02) 0.0 5.0) :MatchedTagName (Variables|Default|GetGCN_Destroyed) :Location (Collision|GetActorBounds _buildingactorowner false false) :Normal "0,0,0" :GameplayEffectLevel 1 :AbilityLevel 1 :bReplicateLocationWhenUsingMinimalRepProxy false)))))
    (:else)))
