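; Loot-drop-on-death component for chrome props that swaps a building's normal destruction effect for a chrome one.
; Once playlist data is ready, it waits for the owner's Event.Damage.Died gameplay event.
; On death it clears the death particles on the BuildingProp or BuildingContainer owner.
; It then fires the GCN_Destroyed burst gameplay cue at the bounds origin, passing a magnitude scaled from the box extent length.
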(event EventBeginPlay
  (|CallParentFunction)
  (Class|FortAsyncActionCurrentPlaylistReady|PlaylistDataReadyAsync
    (:OnReady
      (Ability|Async|WaitGameplayEventToActor (Variables|Default|GetBuildingActorOwner) "(TagName=\"Event.Damage.Died\")" true
        (:EventReceived
          ;[1 1 1] Null out the original death FX of props and containers.
          (CallFunction|RemoveDeathFX)
          ;[1 1 1] Play new chrome Death FX
          (bind _buildingactorowner (Variables|Default|GetBuildingActorOwner)
            (:then
              (GameplayCue|ExecuteGameplayCueOnActor(Burst) _buildingactorowner (Variables|Default|GetGCN_Destroyed) (Ability|GameplayCue|MakeGameplayCueParameters :NormalizedMagnitude 0.0 :RawMagnitude (/ (Math|Vector|VectorLength (.BoxExtent (Collision|GetActorBounds _buildingactorowner true false))) 1e+02) :MatchedTagName (Variables|Default|GetGCN_Destroyed) :Location (Collision|GetActorBounds _buildingactorowner true false) :Normal "0,0,0" :GameplayEffectLevel 1 :AbilityLevel 1 :bReplicateLocationWhenUsingMinimalRepProxy false)))
            (:else)))))))

(event Custom|RemoveDeathFX
  ;[1 1 1] Null out the original death FX of props and containers.
  (bind _asbuilding_prop (Utilities|Casting|CastToBuildingProp (Variables|Default|GetBuildingActorOwner))
    (:then
      (Class|BuildingStaticMesh|SetDeathParticles _asbuilding_prop 0))
    (:CastFailed
      (bind _asbuilding_container (Utilities|Casting|CastToBuildingContainer (Variables|Default|GetBuildingActorOwner)))
      (Class|BuildingStaticMesh|SetDeathParticles _asbuilding_container 0))))
